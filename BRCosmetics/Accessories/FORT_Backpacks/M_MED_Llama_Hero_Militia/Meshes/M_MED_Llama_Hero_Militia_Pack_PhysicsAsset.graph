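# Physics asset for the Llama Hero Militia backpack cosmetic, setting up box bodies for the pack and its two dangling pouches.
# A tiny master pouch body links to the pack through a constraint with free swing and twist.
# Both pouch bodies attach to that master with tight swing and twist limits, forming a small branching chain.
# The pack body is a thin, rotated slab.
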PhysicsAsset /BRCosmetics/Accessories/FORT_Backpacks/M_MED_Llama_Hero_Militia/Meshes/M_MED_Llama_Hero_Militia_Pack_PhysicsAsset
bodies {
 dyn_pouch_a [{shapeType=Box, extentX=12.8887, extentY=4.3872, extentZ=7.8681, center=<-0.3771 -2.0067 -0.053>}]
 dyn_pouch_b [{shapeType=Box, extentX=13.6073, extentY=4.144, extentZ=7.6879, center=<-0.2788 -2.4172 0.6013>}]
 dyn_pouch_mstr [{shapeType=Box, extentX=1.01, extentY=1.01, extentZ=1.01}]
 pack [{shapeType=Box, extentX=36.6453, extentY=2.2677, extentZ=36.6453, center=<1.4384 -1.2633 -1.053>, rotation=<rot -37.785 9.9085 -10.3915>}]
}
constraints [
 {
  bone1Name=dyn_pouch_mstr
  bone2Name=pack
  swing1LimitDegrees=45
  swing1Motion=Free
  swing2LimitDegrees=45
  swing2Motion=Free
  twistLimitDegrees=45
  twistMotion=Free
 }
 {bone1Name=dyn_pouch_a, bone2Name=dyn_pouch_mstr, swing1LimitDegrees=10, swing2LimitDegrees=10, twistLimitDegrees=2}
 {bone1Name=dyn_pouch_b, bone2Name=dyn_pouch_mstr, swing1LimitDegrees=10, swing2LimitDegrees=10, twistLimitDegrees=2}
]
